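# Niagara idle effect for the Dash God Medallion pickup: a floating skull-orbit element wreathed in green glow, smoke and energy.
# Skull_Orbit drives a mesh and two sprite layers with a curl-noise pendulum, orienting to smoothed velocity and growing in with system age.
# Energy reads Skull_Orbit particles through an attribute reader to spawn short-lived rising sprites.
# Light adds a pulsing glow sprite and light, while NE_SandSmoke emits slowly rotating, rising smoke.
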NiagaraSystem /SunRoseMedallionsGameplay/FX/Pickup/Idle/NS_GodMedallion_Dash_Idle
dependencies {
 usedDataInterfaces [
  &NiagaraDataInterfaceCurve
  &NiagaraDataInterfaceParticleRead
  &NiagaraDataInterfaceColorCurve
  &NiagaraDataInterfaceVectorCurve
 ]
 usedDynamicInputs [&ClampFloat, &RemapRange]
 usedModules [
  &self:SystemSpawnScript.SystemScriptSource.SystemScriptGraph.NiagaraNodeAssignment_1.SetVariables_#1
  &SystemState
  &self:SystemSpawnScript.SystemScriptSource.SystemScriptGraph.NiagaraNodeAssignment_3.SetVariables_#2
  &self:SystemSpawnScript.SystemScriptSource.SystemScriptGraph.NiagaraNodeAssignment_0.SetVariables_#3
  &self:SystemSpawnScript.SystemScriptSource.SystemScriptGraph.NiagaraNodeAssignment_5.SetVariables_#4
 ]
}
emitters {
 Energy {
  emitterUpdateScript [
   &EmitterState(
    Enable Distance Culling=true
    Life Cycle Mode=::NewEnumerator1
    Loop Duration=1
    MaxDistance=17000
    Module.bUseMaxDistance=true
    Module.bUseMinDistance=false
    Scalability Mode=::NewEnumerator1
    UseLoopDelay=false
   )
   &SpawnParticlesFromOtherEmitter(
    Attribute Reader=DataInterface() {propertyValues {EmitterBinding {bindingMode=Other, emitterName=Skull_Orbit}}}
    Maximum Age Spawn Threshold=0.8
    Minimum Age Spawn Threshold=0.25
    Spawn Rate=6
    Use Maximum Age=false
    Use Minimum Age=false
   )
  ]
  particleSpawnScript [
   &InitializeParticle(
    Color=&MakeCustomLinearColorFromBool(Bool=&RandomBool(), True Color=<rgba 0.0904 0.6198 0.3551 0.2>, False Color=<rgba 0.0145 0.6979 0.3562 0.2>)
    Lifetime=&RandomRangeFloat(Minimum=0.8, Maximum=1.5)
    Sprite Rotation Mode=::NewEnumerator2
    Sprite Size Mode=::NewEnumerator1
    Uniform Sprite Size Max=45
    Uniform Sprite Size Min=30
    UsePositionOffset=false
    Write Material Random=true
   )
   &SampleParticlesFromOtherEmitter(
    Dynamic Material Parameter 3 Sampling:ENiagara_DirectReadSamplingMode=::NewEnumerator0
    Mesh Orientation Sampling:ENiagara_DirectReadSamplingMode=::NewEnumerator1
   )
   &DynamicMaterialParameters(
    Index 0 Param 1=&RandomRangeFloat(Minimum=0, Maximum=1)
    Module.Param0WriteEnabled=true
    Module.Param1WriteEnabled=true
    Module.Param2WriteEnabled=true
    Module.Param3WriteEnabled=true
   ) {module=DynamicMaterialParameters001}
   &CameraOffset()
   &AddVelocity(Velocity=<0 0 40>, Velocity Speed Scale=&RandomRangeFloat(Minimum=0.6, Maximum=1)) {module=AddVelocity001}
  ]
  particleUpdateScript [
   &ParticleState()
   &ScaleSpriteSize(Scale Factor=&Multiply_Vector2(
     A=<1 1>
     B=&Vector2DFromFloat(Value=&FloatFromCurve(
       FloatCurve=DataInterface() {propertyValues {Curve {keys [
           {arriveTangent=24.2191, interpMode=RCIM_Cubic, leaveTangent=24.2191, tangentMode=RCTM_User, time=0, value=0.5}
           {arriveTangent=1.1179, interpMode=RCIM_Cubic, leaveTangent=1.1179, tangentMode=RCTM_User, time=0.2, value=1.8}
           {arriveTangent=-1.9107, interpMode=RCIM_Cubic, leaveTangent=-1.9107, tangentMode=RCTM_User, time=1, value=0}
          ]}}}
       Scale Curve=&FloatFromCurve(
        FloatCurve=DataInterface() {propertyValues {Curve {keys [
            {interpMode=RCIM_Cubic, tangentMode=RCTM_User, time=0, value=0}
            {arriveTangent=2.0819, interpMode=RCIM_Cubic, leaveTangent=2.0819, tangentMode=RCTM_User, time=0.5, value=1}
           ]}}}
        CurveIndex=$Engine.System.Age
        Scale Curve=0.5
       )
      ))
    ))
   &ScaleColor(
    Linear Color Curve=DataInterface() {propertyValues {
      AlphaCurve {keys [<at 0.0078 0>,<at 0.0872 2>,<at 0.1609 1>,<at 0.4571 1>,<at 1 0>]}
      BlueCurve {keys [<at 0 2>,<at 0.72 0.4555>,<at 0.9893 0.5521>]}
      GreenCurve {keys [<at 0 2>,<at 0.72 0.5521>,<at 0.9893 0.0807>]}
      RedCurve {keys [<at 0 2>,<at 0.72 0.4417>,<at 0.9893 0>]}
     }}
    Scale Mode=::NewEnumerator2
   ) {module=ScaleColor001}
   &Drag(Drag=2, Rotational Drag=5, Use Linear Drag=true, Use Rotational Drag=true)
   &SolveForcesAndVelocity(Clamp Velocity=false, Limit Acceleration=false)
  ]
  rendererClasses [&NiagaraSpriteRendererProperties]
  renderers [{rendererClass=&NiagaraSpriteRendererProperties}]
 }
 Light {
  emitterUpdateScript [
   &EmitterState(Life Cycle Mode=::NewEnumerator1, Loop Behavior=::NewEnumerator1, Loop Duration Mode=::NewEnumerator1, UseLoopDelay=false)
   &SpawnBurst_Instantaneous(Use Spawn Probability=false, UseLoopCountLimit=false)
  ]
  particleSpawnScript [
   &InitializeParticle(
    Color=<rgba 0.091 1.3 0.2925 0.7>
    Sprite Size Mode=::NewEnumerator3
    Uniform Sprite Size=230
    UsePositionOffset=false
    Write Material Random=true
   )
   &SystemLocation(
    MaskSpawnGroup=false
    Offset=&AddVectorToPosition(Position=Linked() {linkedVariable {name=System.PSystemPosition, type=Vector3f}}, Vector=<0 0 0>)
   )
   &CameraOffset(Camera Offset Amount=-30)
  ]
  particleUpdateScript [
   &ParticleState(Kill Particles When Lifetime Has Elapsed=false, Let Infinitely Lived Particles Die When Emitter Deactivates=true)
   &ScaleColor(
    Scale Alpha=&Waveform("Waveform [1]"=::NewEnumerator0, "[1] Frequency"=0.3, "[1] Amplitude Min/Max"=<0.3 1>)
    ScaleA=true
    ScaleRGB=true
   )
   &DynamicMaterialParameters(
    Index 0 Param 1=1
    Index 0 Param 2=10
    Module.Index0Param0Write=true
    Module.Index0Param1Write=true
    Module.Index0Param2Write=false
    Module.Index0Param3Write=false
   )
   &FN_Light_Attributes(
    Color=$Particles.Color
    Create Color Attribute=true
    Create Position Attribute=true
    Position=&AddVectorToPosition(Position=Linked() {linkedVariable {name=System.PSystemPosition, type=Vector3f}}, Vector=<0 0 -40>)
    Scale Color=3
   )
  ]
  rendererClasses [&NiagaraSpriteRendererProperties, &NiagaraLightRendererProperties]
  renderers [
   {rendererClass=&NiagaraSpriteRendererProperties}
   {rendererClass=&NiagaraLightRendererProperties, rendererIndex=1}
  ]
 }
 NE_SandSmoke {
  emitterUpdateScript [
   &EmitterState()
   &SpawnRate(Spawn Probability=0.8, SpawnRate=7, Use Spawn Probability=true)
  ]
  particleSpawnScript [
   &InitializeParticle(
    Color=<rgba 0.1 1 0.475 0.4>
    Lifetime Mode=::NewEnumerator1
    Mass Mode=::NewEnumerator3
    Sprite Rotation Mode=::NewEnumerator2
    Sprite Size Mode=::NewEnumerator1
    Uniform Sprite Size Max=85
    Uniform Sprite Size Min=65
    UsePositionOffset=false
    Write Material Random=true
   )
   &ShapeLocation(MaskSpawnGroup=false, Offset Mode=::NewEnumerator0, Override Seed=false, Sphere Radius=13, UseOwnerScale=true)
   &CameraOffset(Camera Offset Amount=30)
   &AddVelocity(Module.Use Falloff Distance=false, Override Seed=false, Velocity Mode=::NewEnumerator1, Velocity Speed=8)
   &DynamicMaterialParameters(
    Index 0 Param 4=0.99
    Module.Index0Param0Write=false
    Module.Index0Param1Write=true
    Module.Index0Param2Write=false
    Module.Index0Param3Write=true
   )
  ]
  particleUpdateScript [
   &ParticleState()
   &SpriteRotationRate(Rotation Rate=&FloatFromCurve(
     FloatCurve=DataInterface() {propertyValues {Curve {keys [
         {arriveTangent=1.4306, interpMode=RCIM_Cubic, leaveTangent=1.4306, tangentMode=RCTM_User, time=0.2, value=0}
         {arriveTangent=1.1179, interpMode=RCIM_Cubic, leaveTangent=1.1179, tangentMode=RCTM_User, time=0.4, value=0.6228}
         {arriveTangent=0.0224, interpMode=RCIM_Cubic, leaveTangent=0.0224, tangentMode=RCTM_User, time=1, value=1}
        ]}}}
     Scale Curve=&RandomRangeFloat(Minimum=-90, Maximum=-30)
    ))
   &ScaleSpriteSize(
    Scale Sprite Size Mode=::NewEnumerator0
    Uniform Scale Factor=&Multiply_Float(
     A=&FloatFromCurve(FloatCurve=DataInterface() {propertyValues {Curve {keys [
          {arriveTangent=2.0751, interpMode=RCIM_Cubic, leaveTangent=2.0751, tangentMode=RCTM_User, time=0, value=0.2}
          {interpMode=RCIM_Cubic, tangentMode=RCTM_User, time=1, value=1}
         ]}}})
     B=1
    )
   )
   &ScaleColor(
    Scale Alpha=&FloatFromCurve(FloatCurve=DataInterface() {propertyValues {Curve {keys [
         {interpMode=RCIM_Cubic, tangentMode=RCTM_User, time=0, value=0}
         {interpMode=RCIM_Cubic, tangentMode=RCTM_User, time=0.4, value=1}
         {arriveTangent=-0.1502, interpMode=RCIM_Cubic, leaveTangent=-0.1502, tangentMode=RCTM_User, time=1, value=0}
        ]}}})
    Scale RGB=&VectorFromFloat(Value=&FloatFromCurve(FloatCurve=DataInterface() {propertyValues {Curve {keys [
          {interpMode=RCIM_Cubic, tangentMode=RCTM_User, time=0.2, value=1}
          {arriveTangent=-3.6699, interpMode=RCIM_Cubic, leaveTangent=-3.6699, tangentMode=RCTM_User, time=0.7, value=0}
         ]}}}))
    ScaleA=true
    ScaleRGB=true
   )
   &AccelerationForce(Acceleration=&RandomRangeVector(Minimum=<0 0 70>, Maximum=<0 0 10>))
   &Drag(Drag=3, Use Linear Drag=true, Use Rotational Drag=true)
   &SolveForcesAndVelocity(Clamp Velocity=false, Limit Acceleration=false)
  ]
  rendererClasses [&NiagaraSpriteRendererProperties]
  renderers [{rendererClass=&NiagaraSpriteRendererProperties}]
 }
 Skull_Orbit {
  emitterUpdateScript [
   &EmitterState(Life Cycle Mode=::NewEnumerator1, Loop Behavior=::NewEnumerator1, Loop Duration Mode=::NewEnumerator1, UseLoopDelay=false)
   &SpawnBurst_Instantaneous(Use Spawn Probability=false, UseLoopCountLimit=false)
  ]
  particleSpawnScript [
   &InitializeParticle(
    Color=<rgba 0 1.4733 2.6 1.6>
    Lifetime=2
    Mesh Scale Mode=::NewEnumerator3
    Mesh Uniform Scale=1.2
    Sprite Size=<28 20>
    Sprite Size Mode=::NewEnumerator4
    UsePositionOffset=false
    Write Material Random=true
   )
   &InitialMeshOrientation(Mesh Axis To Orient=<0 1 0>, Mesh Orientation Mode=::NewEnumerator2, Use Rotation Vector=false)
   &CameraOffset(Camera Offset Amount=10)
   /SetVariables(
    Particles.Color_Sprite_Dark:LinearColor=<rgba 0 0 0 0.75>
    Particles.Sprite_Size_Dark:Vector2f=&Multiply_Vector2D_ByFloat(Float=0.85, Vector2D=<20 20>)
   ) {isSetParameters=true}
   /SetVariables(
    Particles.Color_Sprite_Light:LinearColor=<rgba 0.24 1.2 0.72 0.1>
    Particles.Sprite_Size_Light:Vector2f=&Multiply_Vector2D_ByFloat(Float=0.85, Vector2D=<20 20>)
   ) {isSetParameters=true}
   /SetVariables(Particles.DynamicParametersSprite:Vector4f=<25 0 0 0>, Particles.DynamicParametersSpriteDark:Vector4f=<0 1 0 0>) {isSetParameters=true}
   &ApplyInitialForces()
  ]
  particleUpdateScript [
   &ParticleState(Kill Particles When Lifetime Has Elapsed=false, Let Infinitely Lived Particles Die When Emitter Deactivates=true)
   &CurlNoiseForce(
    Noise Frequency=0.3
    Noise Quality / Cost=::NewEnumerator0
    Pan Noise=true
    Pan Noise Field=&VectorFromFloat(Value=0.1)
    Randomize Noise Sample=false
    Write to Intrinsic Parameters=false
   )
   /SetVariables(Particles.LastCurlNoiseFOrce:Vector3f=$Output.CurlNoiseForce.CurlNoiseForce) {isSetParameters=true}
   &PendulumSetup(
    Module.Enable Spring Driver=false
    Pendulum Length=7
    Pendulum Pivot:Vector3f=&AddVectorToPosition(Position=<0 0 0>, Vector=<0 0 35>)
    Pendulum Pivot Offset:Vector3f=/Niagara/DynamicInputs/Lerp/SmoothLerpOverTime_Vector(
     Convergence Rate:NiagaraFloat=1
     Smooth Value:Vector3f=$Output.CurlNoiseForce.CurlNoiseForce
     Target Value:Vector3f=$Particles.LastCurlNoiseFOrce
    )
    Pendulum Pivot Offset Coordinate Space:ENiagaraCoordinateSpace=::Local
   )
   &InterpolateOverTime(
    Initial Vector Value:Vector3f=$Particles.Previous.Velocity
    Rate Of Change=4
    Target Vector Value:Vector3f=$StackContext.Velocity
    Variable Type=::NewEnumerator2
   ) {module=InterpolateOverTime001}
   /Niagara/Modules/Update/Orientation/V2/OrientMeshToVector(
    Look At Direction:Vector3f=&NormalizeVector(
     Vector To Normalize=$Particles.InterpolateOverTime001.InterpolateOverTime_V3_.Moving Average
     Safe Normalize=true
     Safe Fallback Threshold=0
     Safe Fallback Vector=<0 0 0.001>
    )
    Override Quat=false
    Up Axis Direction:Vector3f=<0 0 1>
    UseOptionalSideAxis=false
    UseOptionalUpAxis=true
   )
   /SetVariables(
    Particles.Color_Sprite_Dark:LinearColor=&MultiplyLinearColor(
     Linear Color=$Particles.Initial.Color_Sprite_Dark
     Scale Factor=&MakeFloat4FromVectorAndFloat(
      "Vector (RGB)"=<1 1 1>
      "Float (Alpha)"=&FloatFromCurve(FloatCurve=DataInterface() {propertyValues {Curve {keys [
           {interpMode=RCIM_Cubic, tangentMode=RCTM_User, time=0, value=0.75}
           {interpMode=RCIM_Cubic, tangentMode=RCTM_User, time=0.5, value=1}
           {tangentMode=RCTM_User, time=1, value=1}
          ]}}})
     )
    )
    Particles.Sprite_Size_Dark:Vector2f=&Multiply_Vector2D_ByFloat(
     Float=&Multiply_Float(
      A=&FloatFromCurve(
       FloatCurve=DataInterface() {propertyValues {Curve {keys [
           {interpMode=RCIM_Cubic, tangentMode=RCTM_User, time=0, value=0}
           {arriveTangent=1.9474, interpMode=RCIM_Cubic, leaveTangent=1.9474, tangentMode=RCTM_User, time=1, value=1}
          ]}}}
       CurveIndex=$Engine.System.Age
      )
      B=&FloatFromCurve(FloatCurve=DataInterface() {propertyValues {Curve {keys [
           {interpMode=RCIM_Cubic, tangentMode=RCTM_User, time=0, value=0.75}
           {interpMode=RCIM_Cubic, tangentMode=RCTM_User, time=0.5, value=1}
           {tangentMode=RCTM_User, time=1, value=1}
          ]}}})
     )
     Vector2D=$Particles.Initial.Sprite_Size_Dark
    )
   ) {isSetParameters=true}
   /SetVariables(
    Particles.Color_Sprite_Light:LinearColor=&MultiplyLinearColor(
     Linear Color=$Particles.Initial.Color_Sprite_Light
     Scale Factor=&MakeFloat4FromVectorAndFloat(
      "Vector (RGB)"=<1 1 1>
      "Float (Alpha)"=&FloatFromCurve(FloatCurve=DataInterface() {propertyValues {Curve {keys [
           {interpMode=RCIM_Cubic, tangentMode=RCTM_User, time=0, value=0.75}
           {interpMode=RCIM_Cubic, tangentMode=RCTM_User, time=0.5, value=1}
           {tangentMode=RCTM_User, time=1, value=1}
          ]}}})
     )
    )
    Particles.Sprite_Size_Light:Vector2f=&Multiply_Vector2D_ByFloat(
     Float=&Multiply_Float(
      A=&FloatFromCurve(
       FloatCurve=DataInterface() {propertyValues {Curve {keys [
           {interpMode=RCIM_Cubic, tangentMode=RCTM_User, time=0, value=0}
           {arriveTangent=1.9474, interpMode=RCIM_Cubic, leaveTangent=1.9474, tangentMode=RCTM_User, time=1, value=1}
          ]}}}
       CurveIndex=$Engine.System.Age
      )
      B=&FloatFromCurve(FloatCurve=DataInterface() {propertyValues {Curve {keys [
           {interpMode=RCIM_Cubic, tangentMode=RCTM_User, time=0, value=0.75}
           {interpMode=RCIM_Cubic, tangentMode=RCTM_User, time=0.5, value=1}
           {tangentMode=RCTM_User, time=1, value=1}
          ]}}})
     )
     Vector2D=$Particles.Initial.Sprite_Size_Light
    )
   ) {isSetParameters=true}
   &ScaleMeshSize(Scale Factor=&VectorFromCurve(
     VectorCurve=DataInterface() {propertyValues {
       XCurve {keys [
         {interpMode=RCIM_Cubic, tangentMode=RCTM_User, time=0, value=0}
         {arriveTangent=2.343, interpMode=RCIM_Cubic, leaveTangent=2.343, tangentMode=RCTM_User, time=1, value=1}
        ]}
       YCurve {keys [
         {interpMode=RCIM_Cubic, tangentMode=RCTM_User, time=0, value=0}
         {arriveTangent=2.343, interpMode=RCIM_Cubic, leaveTangent=2.343, tangentMode=RCTM_User, time=1, value=1}
        ]}
       ZCurve {keys [
         {interpMode=RCIM_Cubic, tangentMode=RCTM_User, time=0, value=0}
         {arriveTangent=2.343, interpMode=RCIM_Cubic, leaveTangent=2.343, tangentMode=RCTM_User, time=1, value=1}
        ]}
      }}
     CurveIndex=$System.Age
     Scale Curve=<0.7 0.7 0.7>
    ))
   &SolveForcesAndVelocity(Clamp Velocity=false, Limit Acceleration=false)
   &DynamicMaterialParameters(
    Index 0 Param 1=1
    Index 1 Param 1=&RandomRangeFloat(Minimum=0.2, Maximum=0.2)
    Index 3 Param 1=&MakeFloatFromInt(Int=$Particles.VisibilityTag)
    Index 3 Param 2=&MakeFloatFromVector(Vector=$Particles.InterpolateOverTime001.InterpolateOverTime_V3_.Moving Average, Channel=::NewEnumerator0)
    Index 3 Param 3=&MakeFloatFromVector(Vector=$Particles.InterpolateOverTime001.InterpolateOverTime_V3_.Moving Average, Channel=::NewEnumerator1)
    Index 3 Param 4=&MakeFloatFromVector(Vector=$Particles.InterpolateOverTime001.InterpolateOverTime_V3_.Moving Average, Channel=::NewEnumerator2)
    Module.Index0Param0Write=true
    Module.Index0Param1Write=true
    Module.Index0Param2Write=true
    Module.Index0Param3Write=true
    Module.Index1Param0Write=true
    Module.Index1Param1Write=true
    Module.Index1Param2Write=true
    Module.Index1Param3Write=true
    Module.Index2Param0Write=true
    Module.Index2Param1Write=true
    Module.Index2Param2Write=true
    Module.Index2Param3Write=false
    Module.Index3Param0Write=true
    Module.Index3Param1Write=true
    Module.Index3Param2Write=true
    Module.Index3Param3Write=true
    Write Parameter Index 1=true
    Write Parameter Index 2=true
    Write Parameter Index 3=true
   )
   &PendulumConstraint()
   &CalculateAccurateVelocity()
   &InterpolateOverTime(Target Quat Value:Quat4f=$Particles.MeshOrientation, Variable Type=::NewEnumerator5) {module=InterpolateOverTime003}
   /SetVariables(Particles.SmoothedOrientation:Quat4f=$StackContext.InterpolateOverTime003.InterpolateOverTime_Quaternion.Moving Average) {isSetParameters=true}
  ]
  rendererClasses [&NiagaraMeshRendererProperties, &NiagaraSpriteRendererProperties]
  renderers [
   {rendererClass=&NiagaraMeshRendererProperties, rendererIndex=0}
   {rendererClass=&NiagaraSpriteRendererProperties, rendererIndex=1}
   {rendererClass=&NiagaraSpriteRendererProperties, rendererIndex=2}
  ]
 }
}
summary {
 systemName=NS_GodMedallion_Dash_Idle
 emitters [
  {emitterName=Light, rendererClasses [&NiagaraSpriteRendererProperties, &NiagaraLightRendererProperties]}
  {emitterName=NE_SandSmoke, rendererClasses [&NiagaraSpriteRendererProperties]}
  {emitterName=Energy, rendererClasses [&NiagaraSpriteRendererProperties]}
  {emitterName=Skull_Orbit, rendererClasses [&NiagaraMeshRendererProperties, &NiagaraSpriteRendererProperties]}
 ]
 userVariables [
  User.BPCallback {
   type=Object
   description=""
   defaultValue {
    struct=NiagaraExt_VariableValue_Object
    value {object=/SunRoseEnvGameplay/GameplayCues/GCNL_SunRose_SilkyWolf_Ready, objectClass=&Object}
   }
  }
  User.Index {type=NiagaraInt32}
  User.Shields {type=NiagaraFloat, defaultValue=25, description=""}
  User.Skull_Charge_Counter {type=NiagaraInt32, defaultValue=3}
 ]
}
